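# Two-sided lit material for flower mesh particles on the Venus pickaxe, blending different color pairs per face through a packed smoke texture.
# Vertex offset rotates about local Y with corrected normals, though the angle, dynamic WPO times vertex green minus itself, cancels to zero.
# SimpleGrassWind adds U-weighted sway except on mobile.
# A CameraShrink static switch pulls vertices toward the object position when the particle nears the camera.
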Material /BRCosmetics/Weapons/FORT_Melee/Pickaxe_Venus/M_BadFlower
graph {
 properties {
  bAutomaticallySetUsageInEditor=false
  bUsedWithMeshParticles=true
  bUsedWithNiagaraMeshParticles=true
  pixelDepthOffsetMode=PDOM_Legacy
  twoSided=true
  usedShadingModels=Default Lit
 }
 nodes [
  Constant() {r=0.1}
  Constant() {r=0.5}
  Constant() {r=0.15}
  n1 = TextureSample(UVs=TextureCoordinate() {vTiling=2}) {texture=/Game/Effects/Fort_Effects/Textures/Smoke/T_ShotGun_Trace_Packed}
  If(
   A=TwoSidedSign()
   "A > B"=LinearInterpolate(A=Constant3Vector() {constant=<rgba 0.0685 0.117 0.1095 0>}, B=Constant3Vector() {constant=<rgba 0.35 0.3229 0.0528 1>}, Alpha=@n1.R)
   "A < B"=LinearInterpolate(A=Constant3Vector() {constant=<rgba 0.8385 0.1703 0.1703 0>}, B=Constant3Vector() {constant=<rgba 1 0 0 0>}, Alpha=@n1.R)
  )
  n2 = DynamicParameter() {defaultValue=<rgba 0 1 1 1>, paramNames [WPO,Param2,Param3,Param4]}
  Multiply(A=@n2.WPO) {constB=0.85}
  n3 = ComponentMask(None=TextureCoordinate()) {r=true}
  OneMinus(None=@n3)
  n4 = Transform(None=Constant3Vector() {constant=<rgba 0 1 0 0>}) {transformSourceType=TRANSFORMSOURCE_Local}
  n5 = VertexColor()
  n6 = Multiply(A=@n2.WPO, B=Subtract(A=@n5.G, B=@n5.G))
  Reroute(None=MaterialFunctionCall(Rotation Axis=@n4, Rotation Angle=@n6) {materialFunction=&FixRotateAboutAxisNormals}.New World Space Vertex Normal)
  n7 = RotateAboutAxis(
   NormalizedRotationAxis=@n4
   RotationAngle=@n6
   PivotPoint=TransformPosition(None=Constant3Vector()) {transformType=TRANSFORMPOSSOURCE_World}
   Position=WorldPosition().XYZ
  )
  n8 = ShadingPathSwitch(
   Default=MaterialFunctionCall!(WindIntensity=@n3, WindWeight=Constant() {r=0.1}, WindSpeed, AdditionalWPO=@n7) {materialFunction=/Engine/Functions/Engine_MaterialFunctions01/WorldPositionOffset/SimpleGrassWind}.Result
   Mobile=@n7
  ) {desc=Migrated from FeatureLevelSwitch}
  n9 = Subtract(A=ObjectPositionWS(), B=WorldPosition().XYZ)
  n10 = Saturate(None=SphereMask(A=ParticlePositionWS(), B=CameraPositionWS(), Radius=Constant() {r=150}, Hardness=Constant() {r=0.1}))
  StaticSwitchParameter(True=LinearInterpolate(A=@n8, B=Multiply(A=@n9, B=@n10) {constB=0.5}, Alpha=@n10), False=@n8) {parameterName=CameraShrink}
 ]
}
