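# Masked, lit vine material for the Venus pickaxe, usable on Cascade and Niagara mesh particles.
# A tiled stripe mask plus a dynamic-parameter bias blends dark green to yellow, alongside a matching normal map.
# A UV-weighted cross-product offset, scaled by WPOScale, feeds SimpleGrassWind on the default path and stands alone on mobile.
# WindIntensity drives both wind intensity and speed, leaving the WindSpeed parameter unused.
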Material /BRCosmetics/Weapons/FORT_Melee/Pickaxe_Venus/M_Vines
graph {
 properties {
  bUsedWithMeshParticles=true
  bUsedWithNiagaraMeshParticles=true
  blendMode=BLEND_Masked
  opacityMaskClipValue=0.25
  pixelDepthOffsetMode=PDOM_Legacy
  usedShadingModels=Default Lit
 }
 nodes [
  Constant() {r=0.2}
  Constant() {r=0.5}
  Constant() {r=0.1}
  n1 = DynamicParameter() {defaultValue=<rgba 1 20 1 1>, paramNames [WindIntensity, WPOScale, WindSpeed, Colorbais]}
  n2 = TextureCoordinate() {uTiling=6}
  n3 = TextureSample(UVs=@n2) {texture=&/T_Strupe_Mask}
  Reroute(None=LinearInterpolate(
    A=Constant3Vector() {constant=<rgba 0.0685 0.117 0.1095 1>}
    B=Constant3Vector() {constant=<rgba 0.35 0.3229 0.0528 1>}
    Alpha=Add(A=@n1.Colorbais, B=@n3.R)
   ))
  Reroute(None=@n3.R)
  Reroute(None=TextureSample(UVs=@n2) {samplerType=SAMPLERTYPE_Normal, texture=&/T_Strupe_Normal}.RGB)
  n4 = DynamicParameter() {defaultValue=<rgba 1 20 1 1>, paramNames [WindIntensity, WPOScale, WindSpeed, Colorbais]}
  n5 = TextureCoordinate() {vTiling=2}
  n6 = ComponentMask(None=@n5) {r=true}
  n7 = Constant3Vector() {constant=<rgba 1 0 0 0>}
  n8 = Constant3Vector() {constant=<rgba 0 0 1 0>}
  n9 = ComponentMask(None=Subtract(A=@n5)) {g=true}
  n10 = LinearInterpolate(Alpha=@n6) {constA=1, constB=0}
  n11 = Multiply(
   A=CrossProduct(A=Transform(None=@n7), B=Transform(None=@n8) {transformSourceType=TRANSFORMSOURCE_View})
   B=Multiply(A=Multiply(A=@n9, B=@n10) {constB=50}, B=@n4.WPOScale) {constB=20}
  )
  ShadingPathSwitch(
   Default=MaterialFunctionCall!(WindIntensity=@n4.WindIntensity, WindWeight=@n6, WindSpeed=@n4.WindIntensity, AdditionalWPO=@n11) {materialFunction=/Engine/Functions/Engine_MaterialFunctions01/WorldPositionOffset/SimpleGrassWind}.Result
   Mobile=@n11
  ) {desc=Migrated from FeatureLevelSwitch}
 ]
}
