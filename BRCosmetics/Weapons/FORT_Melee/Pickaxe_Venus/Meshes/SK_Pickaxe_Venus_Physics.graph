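# Physics asset for the Venus pickaxe skeletal mesh, giving its dynamic vine and leaf bones simple capsule and box bodies.
# Two three-segment vine chains, a two-segment forward leaf chain, a rear leaf and five individual leaves all constrain back to the head_master body.
# Vine, forward-leaf and rear-leaf joints lock both swing axes, so their swing angle values are inactive; the five individual leaves leave swing unlocked.
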PhysicsAsset /BRCosmetics/Weapons/FORT_Melee/Pickaxe_Venus/Meshes/SK_Pickaxe_Venus_Physics
bodies {
 dyn_fwd_leafA_01 [{shapeType=Capsule, length=13.83, radius=5.9453, center=<1.0871 10.8333 -0.6134>, rotation=<rot 0 170.8821 -88.6169>}]
 dyn_fwd_leafA_02 [{shapeType=Capsule, length=5.7424, radius=5.0259, center=<-0.1525 8.2474 -2.6424>, rotation=<rot -0.1554 0.8811 -72.3222>}]
 dyn_large_vineA_01 [{shapeType=Capsule, length=10.0113, radius=3.172, center=<1.6957 -1.5025 -5.3226>, rotation=<rot 17.4863 0.7138 -164.7297>}]
 dyn_large_vineA_02 [{shapeType=Capsule, length=7.3949, radius=2.0257, center=<0.1648 -3.49 2.7384>, rotation=<rot 0 3.3033 -53.1284>}]
 dyn_large_vineA_03 [{shapeType=Box, extentX=2.7049, extentY=9.9573, extentZ=5.9161, center=<0 -6.3139 0.7066>}]
 dyn_large_vineB_01 [{shapeType=Capsule, length=11.8097, radius=4.3423, center=<-3.5101 -6.081 4.0131>, rotation=<rot -4.9527 -25.3844 -60.5631>}]
 dyn_large_vineB_02 [{shapeType=Capsule, length=7.3685, radius=3.8427, center=<0 -4.6435 2.8194>, rotation=<rot 0 0 -70.0001>}]
 dyn_large_vineB_03 [{shapeType=Capsule, length=7.505, radius=3.255, center=<-0.0003 -4.9108 -2.0466>, rotation=<rot -74.9971 89.9952 179.9999>}]
 dyn_leaf_01 [{shapeType=Capsule, length=0.505, radius=0.505, rotation=<rot 0 -90 0>}]
 dyn_leaf_02 [{shapeType=Box, extentX=7.5, extentY=7.5, extentZ=2.5256, center=<6.0146 -1.252 -0.155>, rotation=<rot -14.9999 0 0>}]
 dyn_leaf_03 [{shapeType=Box, extentX=7.5, extentY=7.5, extentZ=2.4263, center=<7.1585 -5.8174 5.2415>, rotation=<rot -0.0001 -39.9979 0>}]
 dyn_leaf_04 [{shapeType=Capsule, length=10.3098, radius=6.2678, center=<-5.2702 5.3608 2.9175>, rotation=<rot 0 -136.3059 -73.3396>}]
 dyn_leaf_05 [{shapeType=Capsule, length=10.3141, radius=6.2334, center=<-6.1631 -4.5128 2.3927>, rotation=<rot 0 -59.3743 -78.5934>}]
 dyn_rear_leaf [{shapeType=Box, extentX=9.1931, extentY=19.7862, extentZ=4.6359, center=<0.0002 -13.4167 0.2201>, rotation=<rot 0 0 -15.0007>}]
 head_master [{shapeType=Capsule, length=1.75, radius=3.0648, center=<0 0 -5.7857>}]
}
constraints [
 {
  bone1Name=dyn_large_vineA_01
  bone2Name=head_master
  swing1LimitDegrees=15
  swing1Motion=Locked
  swing2LimitDegrees=15
  swing2Motion=Locked
  twistLimitDegrees=45
 }
 {
  bone1Name=dyn_large_vineA_02
  bone2Name=dyn_large_vineA_01
  swing1LimitDegrees=15
  swing1Motion=Locked
  swing2LimitDegrees=15
  swing2Motion=Locked
  twistLimitDegrees=45
 }
 {
  bone1Name=dyn_large_vineA_03
  bone2Name=dyn_large_vineA_02
  swing1LimitDegrees=15
  swing1Motion=Locked
  swing2LimitDegrees=15
  swing2Motion=Locked
  twistLimitDegrees=45
 }
 {
  bone1Name=dyn_large_vineB_01
  bone2Name=head_master
  swing1LimitDegrees=45
  swing1Motion=Locked
  swing2LimitDegrees=45
  swing2Motion=Locked
  twistLimitDegrees=45
 }
 {
  bone1Name=dyn_large_vineB_02
  bone2Name=dyn_large_vineB_01
  swing1LimitDegrees=45
  swing1Motion=Locked
  swing2LimitDegrees=45
  swing2Motion=Locked
  twistLimitDegrees=45
 }
 {
  bone1Name=dyn_large_vineB_03
  bone2Name=dyn_large_vineB_02
  swing1LimitDegrees=45
  swing1Motion=Locked
  swing2LimitDegrees=45
  swing2Motion=Locked
  twistLimitDegrees=45
 }
 {
  bone1Name=dyn_fwd_leafA_01
  bone2Name=head_master
  swing1LimitDegrees=45
  swing1Motion=Locked
  swing2LimitDegrees=45
  swing2Motion=Locked
  twistLimitDegrees=30
 }
 {
  bone1Name=dyn_fwd_leafA_02
  bone2Name=dyn_fwd_leafA_01
  swing1LimitDegrees=45
  swing1Motion=Locked
  swing2LimitDegrees=45
  swing2Motion=Locked
  twistLimitDegrees=30
 }
 {
  bone1Name=dyn_rear_leaf
  bone2Name=head_master
  swing1LimitDegrees=45
  swing1Motion=Locked
  swing2LimitDegrees=45
  swing2Motion=Locked
  twistLimitDegrees=30
 }
 {bone1Name=dyn_leaf_05, bone2Name=head_master, swing1LimitDegrees=10, swing2LimitDegrees=30, twistLimitDegrees=30}
 {bone1Name=dyn_leaf_04, bone2Name=head_master, swing1LimitDegrees=10, swing2LimitDegrees=30, twistLimitDegrees=30}
 {bone1Name=dyn_leaf_03, bone2Name=head_master, swing1LimitDegrees=10, swing2LimitDegrees=30, twistLimitDegrees=30}
 {bone1Name=dyn_leaf_02, bone2Name=head_master, swing1LimitDegrees=10, swing2LimitDegrees=30, twistLimitDegrees=30}
 {bone1Name=dyn_leaf_01, bone2Name=head_master, swing1LimitDegrees=10, swing2LimitDegrees=30, twistLimitDegrees=30}
]
