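# Animation Blueprint for the Venus pickaxe skeleton that adds physics-driven secondary motion to the melee weapon mesh.
# The input pose runs through a RigidBody node using the pickaxe's physics asset, with scaled component linear velocity and acceleration, then goes to the output.
# A chain of three Bone Driven Controllers, copying or mirroring left leaf-bone Y rotation onto other leaf bones, is left unconnected to the output.
# The event graph holds no logic.
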AnimBlueprint /BRCosmetics/Weapons/FORT_Melee/Pickaxe_Venus/Meshes/SK_Pickaxe_Venus_Skeleton_AnimBlueprint
target_skeleton=&/SK_Pickaxe_Venus_Skeleton
graphs [
 AnimGraph {nodes [
   n1 = "dyn_fwd_leafA_02_l.rotateY = dyn_fwd_leafA_01_l.rotateY - Bone Driven Controller"!(ComponentPose:Component Space Pose Link Structure, Alpha:float=0) -> (Pose:Component Space Pose Link Structure) {
    is=title AnimGraphNode_BoneDrivenController
    bAffectTargetRotationY=true
    sourceComponent=RotationY
    alphaBoolBlend {blendOption=Linear}
    sourceBone {boneName=dyn_fwd_leafA_01_l}
    targetBone {boneName=dyn_fwd_leafA_02_l}
   }
   n2 = "dyn_fwd_leafA_01_r.rotateY = dyn_fwd_leafA_01_l.rotateY * -1 - Bone Driven Controller"!(ComponentPose:Component Space Pose Link Structure=@n1, Alpha:float=0) -> (Pose:Component Space Pose Link Structure) {
    is=title AnimGraphNode_BoneDrivenController
    bAffectTargetRotationY=true
    multiplier=-1
    sourceComponent=RotationY
    alphaBoolBlend {blendOption=Linear}
    sourceBone {boneName=dyn_fwd_leafA_01_l}
    targetBone {boneName=dyn_fwd_leafA_01_r}
   }
   "dyn_fwd_leafA_02_r.rotateY = dyn_fwd_leafA_01_l.rotateY * -1 - Bone Driven Controller"!(ComponentPose:Component Space Pose Link Structure=@n2, Alpha:float=0) -> (Pose:Component Space Pose Link Structure) {
    is=title AnimGraphNode_BoneDrivenController
    bAffectTargetRotationY=true
    multiplier=-1
    sourceComponent=RotationY
    alphaBoolBlend {blendOption=Linear}
    sourceBone {boneName=dyn_fwd_leafA_01_l}
    targetBone {boneName=dyn_fwd_leafA_02_r}
   }
   AnimGraphNode_SubInput_0 = Input Pose()
   n3 = RigidBody(ComponentPose=Local To Component(LocalPose=@AnimGraphNode_SubInput_0)) {
    overridePhysicsAsset=&/SK_Pickaxe_Venus_Physics
    alphaBoolBlend {blendOption=Linear}
    componentLinearAccScale=<2 2 2>
    componentLinearVelScale=<5 5 5>
   }
   Output Pose(Result=Component To Local(ComponentPose=@n3)) {layerGroup=None}
  ]}
 EventGraph {nodes [Event BlueprintUpdateAnimation()]}
]
variables []
